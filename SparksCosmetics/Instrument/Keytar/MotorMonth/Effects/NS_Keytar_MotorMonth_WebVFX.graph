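# Niagara system for the MotorMonth keytar cosmetic's glowing green web effect.
# Web_Rings and Web_Spokes are mesh emitters whose bright and dark colors scale by time of day and front-end context, with platform also scaling the dark color.
# A third emitter drives a glow sprite and light from a waveform-based flicker.
# All three read one component of the owner's scale as an alpha multiplier.
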NiagaraSystem /SparksCosmetics/Instrument/Keytar/MotorMonth/Effects/NS_Keytar_MotorMonth_WebVFX
dependencies {
 usedDataInterfaces [&NiagaraDataInterfacePlatformSet]
 usedDynamicInputs []
 usedModules [&SystemState]
}
emitters {
 Default_LightAndGlowFlicker_Loop {
  emitterUpdateScript [
   &EmitterState(
    Inactive Response=::NewEnumerator1
    Life Cycle Mode=::NewEnumerator1
    Loop Behavior=::NewEnumerator1
    Loop Duration Mode=::NewEnumerator1
    UseLoopDelay=false
   )
   &SpawnBurst_Instantaneous(Use Spawn Probability=false, UseLoopCountLimit=false)
   /Niagara/Noise/Waveform(
    EnableExponent1=false
    GlobalClampMax=false
    GlobalClampMin=false
    UseInput1=false
    UsePhase1=false
    "Waveform [1]":ENiagara_Waveforms=::NewEnumerator12
    "[1] Frequency"=7
   )
   /SetVariables(Emitter.Flicker:NiagaraFloat=&Add_Float(A=&Multiply_Float(A=$Emitter.Waveform.Waveform, B=0.5), B=0.5)) {isSetParameters=true}
  ]
  particleSpawnScript [
   &InitializeParticle(
    Color=<rgba 0.5 1.25 0.1 0.5>
    Position=<20 0 11>
    Position Mode=::NewEnumerator1
    Sprite Size Mode=::NewEnumerator3
    Uniform Sprite Size=90
    UsePositionOffset=false
    Write Material Random=true
   )
   &DynamicMaterialParameters(
    Index 0 Param 1=0.2
    Index 0 Param 2=1
    Module.Param0WriteEnabled=true
    Module.Param1WriteEnabled=true
    Module.Param2WriteEnabled=true
    Module.Param3WriteEnabled=true
   )
  ]
  particleUpdateScript [
   &ParticleState(Kill Particles When Lifetime Has Elapsed=false, Let Infinitely Lived Particles Die When Emitter Deactivates=true)
   &ScaleColor(
    Scale Alpha=&Multiply_Float(
     A=&Multiply_Float(A=$Emitter.Flicker, B=0.3)
     B=&MakeFloatFromVector(Vector=$Engine.Owner.Scale, Channel=::NewEnumerator2)
    )
    ScaleA=true
    ScaleRGB=true
   )
   &FN_Light_Attributes(
    Color=$Particles.Color
    Create Color Attribute=true
    Create Exponent Attribute=true
    Create Radius Attribute=true
    Exponent=3.5
    Radius=60
    Scale Color=&Multiply_Float(A=7, B=$Emitter.Flicker)
   )
   &CameraOffset(Camera Offset Amount=10)
   /SetVariables(Particles.SpriteFacing=<0 1 0>) {isSetParameters=true}
  ]
  rendererClasses [&NiagaraSpriteRendererProperties, &NiagaraLightRendererProperties]
  renderers [
   {rendererClass=&NiagaraSpriteRendererProperties}
   {rendererClass=&NiagaraLightRendererProperties, rendererIndex=1}
  ]
 }
 Web_Rings {
  emitterUpdateScript [
   &EmitterState(
    Inactive Response=::NewEnumerator1
    Life Cycle Mode=::NewEnumerator1
    Loop Behavior=::NewEnumerator1
    Loop Duration Mode=::NewEnumerator1
    UseLoopDelay=false
   )
   &SpawnBurst_Instantaneous(Use Spawn Probability=false, UseLoopCountLimit=false)
  ]
  particleSpawnScript [&InitializeParticle(Color=<rgba 0.2902 1 0.2941 1>, Lifetime=10, UsePositionOffset=false, Write Material Random=true)]
  particleUpdateScript [
   &ParticleState(Kill Particles When Lifetime Has Elapsed=false, Let Infinitely Lived Particles Die When Emitter Deactivates=true)
   &ScaleColor(
    Scale Alpha=&Multiply_Float(
     A=&FNTimeOfDayScalar(Night Scale=0.7, Morning Scale=1, Evening Scale=1)
     B=&Multiply_Float(A=&FN_FrontEndFloatModifier(Gameplay=1, Front End=0.7), B=1)
    )
    Scale RGB=<2 5 2>
    ScaleA=true
    ScaleRGB=true
   )
   /SetVariables(Particles.DarkColor:LinearColor=&MultiplyLinearColor(
     Linear Color=<rgba 0.0237 0.1667 0.0208 0.7>
     Scale Factor=&MakeVector4_FromFloat(Float=&Multiply_Float(
       A=&Multiply_Float(
        A=&FNTimeOfDayScalar(Night Scale=0.7, Morning Scale=1, Evening Scale=1)
        B=&Multiply_Float(
         A=&FN_FrontEndFloatModifier(Gameplay=1, Front End=0.6)
         B=&FN_PickFloatDependOnPlatform(TargetPlatformActive=1, TargetPlatformNoActive=0.75)
        )
       )
       B=1
      ))
    )) {isSetParameters=true}
   &ScaleColor(
    Scale Alpha=&MakeFloatFromVector(Vector=$Engine.Owner.Scale, Channel=::NewEnumerator2)
    ScaleA=true
    ScaleRGB=false
   ) {module=ScaleColor001}
  ]
  rendererClasses [&NiagaraMeshRendererProperties]
  renderers [
   {rendererClass=&NiagaraMeshRendererProperties, rendererIndex=0}
   {rendererClass=&NiagaraMeshRendererProperties, rendererIndex=1}
  ]
 }
 Web_Spokes {
  emitterUpdateScript [
   &EmitterState(
    Inactive Response=::NewEnumerator1
    Life Cycle Mode=::NewEnumerator1
    Loop Behavior=::NewEnumerator1
    Loop Duration Mode=::NewEnumerator1
    UseLoopDelay=false
   )
   &SpawnBurst_Instantaneous(Use Spawn Probability=false, UseLoopCountLimit=false)
  ]
  particleSpawnScript [&InitializeParticle(Color=<rgba 0.2902 1 0.2941 1>, Lifetime=10, UsePositionOffset=false, Write Material Random=true)]
  particleUpdateScript [
   &ParticleState(Kill Particles When Lifetime Has Elapsed=false, Let Infinitely Lived Particles Die When Emitter Deactivates=true)
   &ScaleColor(
    Scale Alpha=&Multiply_Float(
     A=&FNTimeOfDayScalar(Night Scale=0.7, Morning Scale=1, Evening Scale=1)
     B=&Multiply_Float(A=&FN_FrontEndFloatModifier(Gameplay=1, Front End=0.7), B=1)
    )
    Scale RGB=<2 5 2>
    ScaleA=true
    ScaleRGB=true
   )
   /SetVariables(Particles.DarkColor:LinearColor=&MultiplyLinearColor(
     Linear Color=<rgba 0.0237 0.1667 0.0208 0.7>
     Scale Factor=&MakeVector4_FromFloat(Float=&Multiply_Float(
       A=&Multiply_Float(
        A=&FNTimeOfDayScalar(Night Scale=0.7, Morning Scale=1, Evening Scale=1)
        B=&Multiply_Float(
         A=&FN_FrontEndFloatModifier(Gameplay=1, Front End=0.6)
         B=&FN_PickFloatDependOnPlatform(TargetPlatformActive=1, TargetPlatformNoActive=0.75)
        )
       )
       B=1
      ))
    )) {isSetParameters=true}
   &ScaleColor(
    Scale Alpha=&MakeFloatFromVector(Vector=$Engine.Owner.Scale, Channel=::NewEnumerator2)
    ScaleA=true
    ScaleRGB=false
   ) {module=ScaleColor001}
  ]
  rendererClasses [&NiagaraMeshRendererProperties]
  renderers [
   {rendererClass=&NiagaraMeshRendererProperties, rendererIndex=0}
   {rendererClass=&NiagaraMeshRendererProperties, rendererIndex=1}
  ]
 }
}
summary {
 systemName=NS_Keytar_MotorMonth_WebVFX
 emitters [
  {emitterName=Web_Spokes, rendererClasses [&NiagaraMeshRendererProperties]}
  {emitterName=Web_Rings, rendererClasses [&NiagaraMeshRendererProperties]}
  {
   emitterName=Default_LightAndGlowFlicker_Loop
   rendererClasses [&NiagaraSpriteRendererProperties, &NiagaraLightRendererProperties]
  }
 ]
 userVariables [
  User.bIsFrontEnd {type=NiagaraBool}
  User.bIsFrontEndPreview {type=NiagaraBool}
  User.StartTime {type=NiagaraFloat, defaultValue=0.2, description=""}
 ]
}
